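# Material function giving the DriftFable cape a comic-book look, zeroing base color, specular and roughness and outputting the stylized result as emissive through LUT and time-of-day adjustments.
# Directional light, camera and fixed-direction dot products drive lit/shaded color blending and hatching masks.
# Procedural multi-level line hatching, Voronoi variation and distance-field lines supply the ink detail.
# FX map channels mask banded metal, curve-atlas chrome, rim light and detail normals; specular is Phong or planar.
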MaterialFunction /BRCosmetics/Accessories/FORT_Capes/Cape_M_MED_DriftFable/Materials/MF_ComicShading_DriftFable
graph {nodes [
  n1 = Saturate(None=LinearInterpolate(A=NamedRerouteUsage(), B=NamedRerouteUsage(), Alpha=NamedRerouteUsage()))
  n2 = StaticSwitchParameter(True=LinearInterpolate(A=@n1, B=NamedRerouteUsage(), Alpha=NamedRerouteUsage()) {desc=Chrome}, False=@n1) {group=01_ComicShading, parameterName=UseChromeShading, sortPriority=0}
  n3 = NamedRerouteUsage()
  n4 = Add(A=LinearInterpolate(A=@n2, B=Multiply(A=@n2, B=@n3), Alpha=NamedRerouteUsage()) {desc=metallic}, B=NamedRerouteUsage())
  n5 = VectorParameter() {group=01_ComicShading, parameterName=RimLight_Color, defaultValue=<rgba 0.3 0.5 1 2>}
  n6 = NamedRerouteUsage()
  n7 = NamedRerouteUsage()
  n8 = Saturate(None=Multiply(A=@n5.A, B=Min(A=@n6, B=@n7)))
  Add(A=@n4, B=@n8)
  n9 = NamedRerouteUsage()
  n10 = TextureSampleParameter2D(UVs=@n9) {
   desc="R=Sheen \r\nG=Theshold\r\nB=Thickness"
   group=01_Anime Maps
   parameterName=STT_Map
   samplerType=SAMPLERTYPE_Masks
   sortPriority=12
   texture=/Game/Global/Textures/Default/Blanks/T_FN_Anime_STT
  }
  ChannelMaskParameter(None=@n10.RGB) {group=01_Anime Maps, maskChannel=Blue, parameterName=STT_Channel_Thickness, sortPriority=23, defaultValue=<rgba 0 0 1 0>}
  n11 = NamedRerouteUsage()
  n12 = Min(A=@n11, B=NamedRerouteUsage())
  n13 = LinearInterpolate(Alpha=MaterialFunctionCall(
    Input=StaticSwitchParameter(True=@n12, False=@n11) {defaultValue=true, group=01_ComicShading, parameterName=UseFixedHatching, sortPriority=0}
    Input Low=ScalarParameter() {group=01_ComicShading, parameterName=Hatching_NdotL_Min, sliderMax=1}
    Input High=ScalarParameter() {defaultValue=0.5, group=01_ComicShading, parameterName=Hatching_NdotL_Max, sliderMax=1}
   ) {materialFunction=&RemapValue_0-1}.Result) {constA=6, constB=0}
  n14 = Constant() {r=1}
  Composite!(In=@n13, RampWidth=@n14) {subgraphName=SmoothRound}
  n15 = Constant()
  n16 = NamedRerouteUsage()
  n17 = SetMaterialAttributes!(
   MaterialAttributes
   BaseColor=@n15
   Specular=@n15
   Opacity=@n15
   SubsurfaceColor=@n15
   Roughness=@n15
   Metallic=NamedRerouteUsage()
   EmissiveColor=MaterialFunctionCall!(Color=MaterialFunctionCall!(In=@n16, LUT, LUT Multiplier) {materialFunction=/Game/Packages/Fortress_SharedMaterials/Anime_Master_Material/MF_ApplyLUT}.Result) {materialFunction=/Game/Packages/Fortress_SharedMaterials/Anime_Master_Material/MF_TODColorAdjustment}.Result
   Normal=NamedRerouteUsage()
   AmbientOcclusion=NamedRerouteUsage()
  ) {attributeSetTypes [
    #1
    #2
    #3
    #4
    #5
    #6
    #7
    #8
    #9
   ]}
  FunctionOutput(None=@n17)
  Multiply(A=@n5.RGB, B=@n5.RGB)
  Multiply(A=MaterialFunctionCall() {materialFunction=&CameraDirectionVector}.Result) {constB=-1}
  n18 = NamedRerouteUsage()
  n19 = NamedRerouteDeclaration(None=DotProduct(A=CameraVectorWS(), B=@n18)) {name=NdotV, nodeColor=<rgba 0 1 0.4249 1>}
  NamedRerouteDeclaration(None=Add(A=Multiply(A=@n19) {constB=0.5}) {constB=0.5}) {name=NdotV_Lambert, nodeColor=<rgba 0 1 0.2727 1>}
  n20 = NamedRerouteUsage()
  n21 = MaterialFunctionCall(BaseNormal=@n20, AdditionalNormal=NamedRerouteUsage()) {materialFunction=&BlendAngleCorrectedNormals}
  n22 = Transform(None=NamedRerouteDeclaration(None=StaticSwitchParameter(True=@n21.Result, False=@n20) {defaultValue=true, group=01_ComicShading, parameterName=UseDetailNormal, sortPriority=0}) {name=N+DN_TS, nodeColor=<rgba 0 0.0305 1 1>})
  n23 = VertexColor()
  n24 = Transform(None=Subtract(A=Multiply(A=@n23) {constB=2})) {transformSourceType=TRANSFORMSOURCE_Local}
  n25 = Multiply(A=ScalarParameter() {group=01_ComicShading, parameterName=VertexBakedNormal_Blend, sliderMax=1}, B=@n23.A)
  NamedRerouteDeclaration(None=Normalize(VectorInput=LinearInterpolate(A=@n22, B=@n24, Alpha=@n25))) {name=WSNormal_DN, nodeColor=<rgba 0 0.4375 1 1>}
  n26 = NamedRerouteDeclaration(None=DotProduct(A=CameraVectorWS(), B=NamedRerouteUsage())) {name=NdotV_NoDN, nodeColor=<rgba 0 1 0.9105 1>}
  NamedRerouteDeclaration(None=Add(A=Multiply(A=@n26) {constB=0.5}) {constB=0.5}) {name=NdotV_Lambert_NoDN, nodeColor=<rgba 0 1 0.9434 1>}
  NamedRerouteDeclaration(None=Transform(None=NamedRerouteUsage())) {name=WSNormal_NoDN, nodeColor=<rgba 0 0.4375 1 1>}
  n27 = Normalize(VectorInput=ReflectionVectorWS(CustomWorldNormal=NamedRerouteUsage()))
  n28 = CurveAtlasRowParameter(CurveTime=ConstantBiasScale(None=ComponentMask(None=@n27) {b=true})) {
   atlas=/Game/Global/Materials/Character_Parent_2023/Comic_Parent_2023/Textures/CA_ToonChromeAtlas
   curve=/Game/Global/Materials/Character_Parent_2023/Comic_Parent_2023/Textures/C_ToonChrome_01
   group=01_ComicShading
  }
  n29 = Power(Base=@n28, Exp=ScalarParameter() {defaultValue=1.2, group=01_ComicShading, parameterName=ChromeExponent})
  n30 = ScalarParameter() {defaultValue=1, group=01_ComicShading, parameterName=ChromeMultiplier}
  NamedRerouteDeclaration(None=Desaturation(None=Multiply(A=@n29, B=@n30), Fraction=ScalarParameter() {defaultValue=-1, group=01_ComicShading, parameterName=ChromeDesaturation})) {name=Chrome_Shading, nodeColor=<rgba 0 0.9766 1 1>}
  n31 = TextureSampleParameter2D(UVs=@n9) {
   group=01_Anime Maps
   parameterName=FX_Map
   samplerType=SAMPLERTYPE_Masks
   sortPriority=16
   texture=&T_BlackMask
  }
  n32 = NamedRerouteDeclaration(None=@n31.RGBA) {name=Map_FX, nodeColor=<rgba 0 0.326 1 1>}
  NamedRerouteDeclaration(None=ChannelMaskParameter(None=@n32) {group=01_Anime Maps, parameterName=FX_Channel_ChromeMask, sortPriority=27}.R) {name=ChromeShadingMask, nodeColor=<rgba 0.2187 1 0 1>}
  n33 = VectorParameter() {group=01_ComicShading, parameterName=FixedHatching_Direction, defaultValue=<rgba -1 -0.5 -1 1>}
  n34 = NamedRerouteDeclaration(None=DotProduct(A=Transform(None=@n33.RGB) {transformSourceType=TRANSFORMSOURCE_Camera}, B=NamedRerouteUsage())) {name=NdotS_NoDN, nodeColor=<rgba 0.875 1 0 1>}
  NamedRerouteDeclaration(None=Add(A=Multiply(A=@n34) {constB=0.5}) {constB=0.5}) {name=NdotS_Lambert_NoDN, nodeColor=<rgba 0.875 1 0 1>}
  n35 = ChannelMaskParameter(None=NamedRerouteUsage()) {
   desc="Select in which channel of the FX map the mask camera light vector will be stored "
   group=01_Anime Maps
   parameterName=CameraLightVector_FXChannel
   sortPriority=25
  }
  n36 = Custom() {code="return View.DirectionalLightDirection;", description=DirectionalLightDirection, inputs []}
  n37 = CollectionParameter() {
   collection=&FortniteMaterialParameters
   parameterName=SunlightDirectionalVector
  }
  n38 = ComponentMask(None=@n37) {b=true, g=true, r=true}
  n39 = If(
   A=ComponentMask(None=Custom() {code="return View.DirectionalLightColor;", description=DirectionalLightColor, outputType=CMOT_Float4, inputs []}) {b=true, g=true, r=true}
   "A > B"=If(A=DotProduct(A=@n36, B=@n36), "A > B"=@n36, "A == B"=@n38, "A < B"=@n38) {desc=check if DirectionalLight exists}
   "A == B"=@n38
   "A < B"=@n38
  ) {desc=check if DirectionalLight color is Black}
  n40 = MaterialFunctionCall(Float3=@n39) {materialFunction=&BreakOutFloat3Components}
  n41 = ScalarParameter() {
   desc="0.0 = the light vector is locked on the XY plane.\r\n1.0 = the actual light vector is used  \r\n(for Faces use a blend value of 0,0)"
   group=01_ComicShading
   parameterName=LockedZLightVector_Blend
   sliderMax=1
  }
  n42 = MaterialFunctionCall(X=@n40.R, Y=@n40.R, Z=Multiply(A=@n40.B, B=@n41)) {materialFunction=&MakeFloat3}
  n43 = StaticSwitchParameter(True=LinearInterpolate(A=@n39, B=Normalize(VectorInput=@n42.Result), Alpha=VertexColor().A), False=@n39) {
   desc="Lock the Z axis of the Light Vector\r\nis driven by the VertexColor A channel (0 = Regular LightVector, 1= Locked LIghtVector)"
   group=01_ComicShading
   parameterName=UseLockedZLightVector
   sortPriority=0
  }
  n44 = StaticSwitchParameter(True=LinearInterpolate(A=@n43, B=CameraVectorWS(), Alpha=@n35.R), False=@n43) {
   desc="Use the Camera vector instead of the light vector which is defined by the FX map. Select the FX channel with this parameter \"CameraLightVector_FXChannel\"."
   group=01_ComicShading
   parameterName=UseCameraLightVector
   sortPriority=0
  }
  NamedRerouteDeclaration(None=@n44) {name=LightVector, nodeColor=<rgba 0.4766 1 0 1>}
  n45 = TextureCoordinate()
  NamedRerouteDeclaration(None=StaticSwitchParameter(
    True=FunctionInput(Preview=@n45) {bUsePreviewValueAsDefault=true, inputName=CustomUVs, inputType=FunctionInput_Vector2, sortPriority=15}
    False=TextureCoordinate()
   ) {group=01_ComicShading, parameterName=UseEyeParallaxFunction, sortPriority=1}) {name=CustomUVs, nodeColor=<rgba 1 0.0993 0 1>}
  n46 = TextureSampleParameter2D(UVs=@n9) {
   group=01_Anime Maps
   parameterName=Normals
   samplerType=SAMPLERTYPE_Normal
   sortPriority=13
   texture=/BRCosmetics/Accessories/FORT_Capes/Cape_M_MED_DriftFable/Textures/T_DriftFable_Cape_N
  }
  NamedRerouteDeclaration(None=@n46.RGB) {name=NormalMap, nodeColor=<rgba 0 0.4375 1 1>}
  n47 = ChannelMaskParameter(None=@n32) {group=01_Anime Maps, maskChannel=Blue, parameterName=FX_Channel_ExcludeHatchMask, sortPriority=28, defaultValue=<rgba 0 0 1 0>}
  NamedRerouteDeclaration(None=@n47.R) {name=ExcludeHatchingMask, nodeColor=<rgba 0 0.8828 1 1>}
  n48 = Transform(None=Constant3Vector() {constant=<rgba 0 0 1 0>}) {transformSourceType=TRANSFORMSOURCE_View, transformType=TRANSFORM_Local}
  n49 = MaterialFunctionCall(Float2=MaterialFunctionCall(Vector or UVs=ComponentMask(None=@n48) {g=true, r=true}) {materialFunction=&VectorToRadialValue}.Radial Coordinates) {materialFunction=&BreakOutFloat2Components}
  n50 = AppendVector(A=Add(A=Multiply(A=@n49.R) {constB=4}) {constB=-1}, B=Multiply(A=ComponentMask(None=@n48) {b=true}) {constB=-1})
  n51 = WorldPosition()
  n52 = MaterialFunctionCall(Float3=Divide(
    A=TransformPosition(None=@n51.XYZ) {transformSourceType=TRANSFORMPOSSOURCE_World}
    B=ScalarParameter() {defaultValue=30, group=01_ComicShading, parameterName=SpecularHighlight_Planar_Scale, sliderMax=200}
   ) {constB=30}) {materialFunction=&BreakOutFloat3Components}
  n53 = MaterialFunctionCall(X=@n52.R, Y=@n52.R) {materialFunction=&MakeFloat2}
  n54 = MaterialFunctionCall(X=@n52.G, Y=@n52.G) {materialFunction=&MakeFloat2}
  n55 = Add(A=Multiply(A=StaticSwitchParameter(True=@n53.Result, False=@n54.Result) {defaultValue=true, group=01_ComicShading, parameterName=UsePlanarReflectionYPlane, sortPriority=2}) {constB=-1}) {constB=0.5}
  n56 = TextureSampleParameter2D(UVs=Rotator(
    Coordinate=Add(A=@n50, B=@n55)
    Time=ScalarParameter() {group=01_ComicShading, parameterName=SpecularHighlight_Planar_Rotate, sliderMax=1, sliderMin=-1}
   ) {speed=3.1415}) {
   group=01_Anime Maps
   parameterName=SpecularMap_Planar
   samplerSource=SSM_Wrap_WorldGroupSettings
   sortPriority=20
   texture=/Game/Global/Materials/Character_Parent_2023/Anime_Parent_2023/Textures/T_Planar_Highlight
  }
  NamedRerouteDeclaration(None=@n56.RGB) {name=PlanarSpec, nodeColor=<rgba 1 0.4327 0 1>}
  n57 = ChannelMaskParameter(None=@n32) {group=01_Anime Maps, maskChannel=Alpha, parameterName=FX_Channel_RimLightMask, sortPriority=26, defaultValue=<rgba 0 0 0 1>}
  NamedRerouteDeclaration(None=@n57.R) {name=RimLightMask, nodeColor=<rgba 1 0.37 0 1>}
  n58 = NamedRerouteDeclaration(None=DotProduct(A=NamedRerouteUsage(), B=@n18)) {name=NdotL, nodeColor=<rgba 0.7344 1 0 1>}
  NamedRerouteDeclaration(None=Add(A=Multiply(A=@n58) {constB=0.5}) {constB=0.5}) {name=NdotL_Lambert, nodeColor=<rgba 0.4531 1 0 1>}
  NamedRerouteDeclaration(None=ChannelMaskParameter(None=@n32) {group=01_Anime Maps, parameterName=FX_Channel_DetailNormalMask, sortPriority=27}.R) {name=DetailNormalMask, nodeColor=<rgba 1 0 1 0>}
  n59 = VectorParameter() {group=01_ComicShading, parameterName=HatchingColor, sortPriority=3, defaultValue=<rgba 0.004 0.004 0.004 1>}
  n60 = LinearInterpolate(A=@n4, B=@n5.RGB, Alpha=@n8)
  n61 = LinearInterpolate(A=@n59.RGB, B=@n60, Alpha=Multiply(A=NamedRerouteUsage(), B=NamedRerouteUsage()))
  n62 = Multiply(A=@n8, B=ScalarParameter() {defaultValue=1, group=01_ComicShading, parameterName=RimLight_HatchingOverlap, sliderMax=1})
  NamedRerouteDeclaration(None=Saturate(None=LinearInterpolate(A=@n61, B=@n60, Alpha=@n62) {desc=RimLight On Top of Hatching})) {name=BaseColor_Raw, nodeColor=<rgba 1 0.4506 0.2387 1>}
  n63 = ChannelMaskParameter(None=@n10.RGB) {group=01_Anime Maps, parameterName=STT_Channel_Threshold, sortPriority=22}
  n64 = ScalarParameter() {defaultValue=1, group=01_ComicShading, parameterName=Threshold_Exponent}
  NamedRerouteDeclaration(None=LinearInterpolate(
    A=ScalarParameter() {group=01_ComicShading, parameterName=Threshold_Min, sliderMax=1}
    B=ScalarParameter() {defaultValue=1, group=01_ComicShading, parameterName=Threshold_Max, sliderMax=1}
    Alpha=Power(Base=@n63.R, Exp=@n64)
   )) {name=Threshold, nodeColor=<rgba 0 0.5 1 1>}
  n65 = Multiply(A=NamedRerouteUsage(), B=ScalarParameter() {defaultValue=0.5, group=01_ComicShading, parameterName=LineDistortion_RandScale, sliderMax=1})
  n66 = NamedRerouteUsage()
  n67 = ScalarParameter() {defaultValue=0.5, group=01_ComicShading, parameterName=LineDistortion_RandRot, sliderMax=0.25}
  n68 = Rotator(
   Coordinate=Multiply(A=TextureCoordinate(), B=Add(A=@n65))
   Time=Add(A=ScalarParameter() {group=01_ComicShading, parameterName=Lines_RotationOffset, sliderMax=1}, B=Multiply(A=@n66, B=@n67))
  ) {centerX=0, centerY=0, speed=-6.283}
  n69 = Frac(None=Multiply(A=ComponentMask(None=@n68) {r=true}, B=ScalarParameter() {defaultValue=180, group=01_ComicShading, parameterName=Lines_Amount}) {constB=32})
  n70 = Abs(None=Add(A=Multiply(A=@n69) {constB=2}) {constB=-1}) {desc=HatchingDF}
  n71 = NamedRerouteUsage()
  NamedRerouteDeclaration(None=Saturate(None=LinearInterpolate(B=@n70, Alpha=@n71) {constA=1})) {name=UVHatching, nodeColor=<rgba 0.9922 1 0 1>}
  n72 = DotProduct(A=ReflectionVectorWS(CustomWorldNormal=NamedRerouteUsage()), B=NamedRerouteUsage()) {desc=Phong}
  n73 = StaticSwitchParameter(
   True=NamedRerouteUsage()
   False=SphereMask(
    A=Saturate(None=@n72)
    B=Constant() {r=1}
    Radius=ScalarParameter() {defaultValue=0.1, group=01_ComicShading, parameterName=Specular_Size, sliderMax=1}
    Hardness=ScalarParameter() {defaultValue=0.8, group=01_ComicShading, parameterName=Specular_Hardness, sliderMax=1}
   ) {hardnessPercent=78}
  ) {group=01_ComicShading, parameterName=UsePlanarSpecular, sortPriority=1}
  n74 = ScalarParameter() {defaultValue=0.5, group=01_ComicShading, parameterName=Specular_Intensity, sliderMax=1}
  n75 = Multiply(A=Saturate(None=Multiply(A=@n73, B=@n74)), B=NamedRerouteUsage())
  n76 = NamedRerouteUsage()
  NamedRerouteDeclaration(None=Multiply(A=Multiply(A=@n75, B=@n76), B=VectorParameter() {group=01_ComicShading, parameterName=Specular_Color, defaultValue=<rgba 1 1 1 1>}.RGB)) {name=Specular_Curve, nodeColor=<rgba 1 0.4453 0 1>}
  n77 = TextureSampleParameter2D(UVs=Multiply(A=TextureCoordinate(), B=ScalarParameter() {defaultValue=128, group=01_ComicShading, parameterName=DetailNormal_Tiling, sliderMax=256})) {
   group=01_Anime Maps
   parameterName=DetailNormal_Map
   samplerType=SAMPLERTYPE_Normal
   sortPriority=19
   texture=/Game/Global/Materials/Character_Parent_2023/Anime_Parent_2023/Textures/T_SurfaceDetail_N
  }
  n78 = ScalarParameter() {group=01_ComicShading, parameterName=DetailNormal_Flattness, sliderMax=2}
  NamedRerouteDeclaration(None=LinearInterpolate(
    A=Constant3Vector() {constant=<rgba 0 0 1 0>}
    B=MaterialFunctionCall(Normal=@n77.RGB, Flatness=@n78) {materialFunction=&FlattenNormal}.Result
    Alpha=NamedRerouteUsage()
   )) {name=DetailNormal, nodeColor=<rgba 0 0.1253 1 1>}
  n79 = ChannelMaskParameter(None=@n10.RGB) {group=01_Anime Maps, maskChannel=Blue, parameterName=STT_Channel_Sheen, sortPriority=21, defaultValue=<rgba 0 0 1 0>}
  NamedRerouteDeclaration(None=@n79.R) {name=Sheen, nodeColor=<rgba 0 0.5 1 1>}
  n80 = Constant() {r=0.5}
  n81 = Multiply(A=Saturate(None=ScalarParameter() {defaultValue=0.05, group=01_ComicShading, parameterName=ShadedBaseColor_TransitionAmount, sliderMax=1})) {constB=0.5}
  n82 = NamedRerouteUsage()
  n83 = ScalarParameter() {defaultValue=-0.3, group=01_ComicShading, parameterName=ShadedBaseColor_Bias, sliderMax=1, sliderMin=-1}
  n84 = Add(A=Multiply(A=Subtract(A=@n82, B=@n83)) {constB=0.5}) {constB=0.5}
  n85 = LinearInterpolate(
   A=@n84
   B=MaterialFunctionCall(Base=NamedRerouteUsage(), Blend=@n84) {materialFunction=&Blend_Overlay}.Result
   Alpha=ScalarParameter() {group=01_ComicShading, parameterName=ShadedBaseColor_ThresholdBlend, sliderMax=1}
  )
  n86 = NamedRerouteUsage()
  NamedRerouteDeclaration(None=SmoothStep(Min=Subtract(A=@n80, B=@n81) {constA=0.5}, Max=Add(A=@n80, B=@n81) {constA=0.5, constB=-0.5}, Value=LinearInterpolate(A=@n85, Alpha=@n86)) {constMax=0.01}) {name=ShadedBaseColor_Curve, nodeColor=<rgba 0 0.9439 1 1>}
  n87 = TextureSampleParameter2D(UVs=@n9) {
   group=01_Anime Maps
   parameterName=DFL_Map
   samplerType=SAMPLERTYPE_Alpha
   sortPriority=14
   texture=/Game/Global/Textures/Default/Blanks/T_BlackAlpha
  }
  NamedRerouteDeclaration(None=StaticSwitchParameter(True=@n87.R, False=Constant()) {defaultValue=true, group=01_ComicShading, parameterName=UseDFLMap, sortPriority=0}) {name=DF, nodeColor=<rgba 0 1 0.6166 1>}
  n88 = ScalarParameter() {defaultValue=1, group=01_DistanceFieldLines, parameterName=DF_Line_GammaCorrection, sliderMax=1, sliderMin=0.4545, sortPriority=5}
  n89 = MaterialFunctionCall!(
   Fade Offset=ScalarParameter() {defaultValue=50, group=01_DistanceFieldLines, parameterName=DF_FadeStart, sliderMax=200}
   Fade Distance=ScalarParameter() {defaultValue=1000, group=01_DistanceFieldLines, parameterName=DF_FadeDistance, sliderMax=1000}
  ) {materialFunction=/Game/Global/Materials/Character_Parent_2023/Anime_Parent_2023/MaterialFunctions/MFC_DistanceFade_FOV_compensation}
  n90 = SphereMask(
   A=Power(Base=NamedRerouteUsage(), Exp=@n88) {constExponent=0.4545, desc=gamma correction}
   B=Constant() {r=1}
   Radius=LinearInterpolate(
    A=ScalarParameter() {defaultValue=0.4, group=01_DistanceFieldLines, parameterName=DF_LineThickness_Near, sliderMax=1, sortPriority=5}
    B=ScalarParameter() {defaultValue=0.6, group=01_DistanceFieldLines, parameterName=DF_LineThickness_Far, sliderMax=1, sortPriority=6}
    Alpha=NamedRerouteDeclaration(None=@n89.Result) {name=DistanceFade, nodeColor=<rgba 1 0 0.4453 1>}
   )
   Hardness=ScalarParameter() {defaultValue=0.9, group=01_DistanceFieldLines, parameterName=DF_LineHardness, sliderMax=1, sortPriority=7}
  ) {attenuationRadius=0.1}
  NamedRerouteDeclaration(None=OneMinus(None=Saturate(None=@n90))) {name=DF_Mask, nodeColor=<rgba 1 0.6562 0 1>}
  n91 = Constant() {r=0.5}
  n92 = Multiply(A=Saturate(None=ScalarParameter() {defaultValue=0.1, group=01_ComicShading, parameterName=LineTransition_Amount, sliderMax=1})) {constB=0.5}
  n93 = Subtract(A=NamedRerouteUsage(), B=ScalarParameter() {group=01_ComicShading, parameterName=ShadingCurve_Bias, sliderMax=1, sliderMin=-1})
  n94 = Min(
   A=@n93
   B=Subtract(
    A=NamedRerouteUsage()
    B=ScalarParameter() {defaultValue=0.1, group=01_ComicShading, parameterName=FixedHatching_ShadingCurve_Bias, sliderMax=1, sliderMin=-1}
   )
  )
  n95 = Add(A=Multiply(A=StaticSwitchParameter(True=@n94, False=@n93) {defaultValue=true, group=01_ComicShading, parameterName=UseFixedHatching, sortPriority=0}) {constB=0.5}) {constB=0.5}
  n96 = Power(
   Base=NamedRerouteUsage()
   Exp=ScalarParameter() {defaultValue=1, group=01_ComicShading, parameterName=ShadingCurve_CameraExponent, sliderMax=32}
  ) {constExponent=4}
  n97 = ScalarParameter() {group=01_ComicShading, parameterName=ShadingCurve_CameraBlend, sliderMax=1}
  n98 = StaticSwitchParameter(True=NamedRerouteUsage(), False=NamedRerouteUsage()) {group=01_ComicShading, parameterName=UseCrossHatching, sortPriority=0}
  n99 = ScalarParameter() {defaultValue=0.5, group=01_ComicShading, parameterName=LineVariation_Intensity, sliderMax=1}
  n100 = Add(
   A=Multiply(A=LinearInterpolate(A=@n95, B=@n96, Alpha=@n97), B=Constant() {r=2})
   B=Subtract(A=Multiply(A=@n98, B=@n99) {constB=0.2}, B=Constant() {r=1})
  )
  n101 = SmoothStep(
   Min=Subtract(A=@n91, B=@n92) {constA=0.5}
   Max=Add(A=@n91, B=@n92) {constA=0.5, constB=-0.5}
   Value=MaterialFunctionCall(Base=NamedRerouteUsage(), Blend=Saturate(None=@n100) {desc=height blend with 1 heighmap}) {materialFunction=&Blend_Overlay}.Result
  ) {constMax=0.01}
  n102 = NamedRerouteUsage()
  n103 = ComponentMask(None=NamedRerouteUsage()) {r=true}
  n104 = Saturate(None=Power(Base=Abs(None=@n103), Exp=Constant() {r=0.4}) {constExponent=3})
  n105 = Normalize(VectorInput=Transform(None=Constant3Vector() {constant=<rgba 0 1 0 0>}) {transformSourceType=TRANSFORMSOURCE_Local})
  n106 = PixelNormalWS()
  n107 = Step(X=Abs(None=DotProduct(A=@n105, B=@n106))) {constX=0.9, constY=0.9}
  n108 = ScalarParameter() {parameterName=IsPickaxe}
  NamedRerouteDeclaration(None=LinearInterpolate(A=LinearInterpolate(A=@n101, Alpha=@n102), Alpha=LinearInterpolate(A=@n104, B=@n107, Alpha=@n108))) {name=ShadingMask, nodeColor=<rgba 1 0.6562 0 1>}
  n109 = SmoothStep(Value=Power(Base=NamedRerouteUsage(), Exp=ScalarParameter() {defaultValue=5, group=01_ComicShading, parameterName=BackLit_Exponent, sliderMax=8}) {constExponent=4}) {constMax=0, constMin=0.02}
  n110 = MaterialFunctionCall() {materialFunction=&CameraDirectionVector}
  n111 = NamedRerouteUsage()
  n112 = Clamp(
   None=Add(
    A=DotProduct(A=@n110.Result, B=@n111)
    B=ScalarParameter() {defaultValue=0.25, group=01_ComicShading, parameterName=BackLit_Bias, sliderMax=1}
   )
   Max=ScalarParameter() {defaultValue=0.75, group=01_ComicShading, parameterName=BackLit_MaxBias, sliderMax=1}
  )
  NamedRerouteDeclaration(None=Multiply(A=Saturate(None=@n109), B=NamedRerouteDeclaration(None=@n112) {name=BackLit_Curve, nodeColor=<rgba 0.9507 0.45 1 1>})) {name=BackLitShading_Curve, nodeColor=<rgba 0.1307 1 0.334 1>}
  n113 = ChannelMaskParameter(None=@n32) {group=01_Anime Maps, maskChannel=Green, parameterName=FX_Channel_MetalMask, sortPriority=25, defaultValue=<rgba 0 1 0 0>}
  NamedRerouteDeclaration(None=@n113.R) {name=MetallicMask, nodeColor=<rgba 0 1 0.4844 1>}
  n114 = MaterialFunctionCall(Float2=TextureCoordinate()) {materialFunction=&BreakOutFloat2Components}
  n115 = Multiply(A=@n114.G, B=Divide(B=ScalarParameter() {defaultValue=1, group=01_ComicShading, parameterName=Voronoi_Ratio, sliderMax=2}))
  n116 = TextureSampleParameter2D(UVs=Multiply(
    A=MaterialFunctionCall(X=@n114.R, Y=@n115) {materialFunction=&MakeFloat2}.Result
    B=ScalarParameter() {defaultValue=2, group=01_ComicShading, parameterName=Voronoi_Tiling, sliderMax=16}
   )) {
   group=01_Anime Maps
   parameterName=VoronoiMap
   samplerType=SAMPLERTYPE_LinearColor
   sortPriority=1
   texture=/Game/Global/Materials/Character_Parent_2023/Anime_Parent_2023/Textures/T_Hatching_Voronoi
  }
  NamedRerouteDeclaration(None=@n116.G) {name=VoronoiID, nodeColor=<rgba 1 0.2603 0.2703 1>}
  NamedRerouteDeclaration(None=SmoothStep(Value=@n116.R) {constMax=0.8, constMin=1}) {name=VoronoiIEdgeDistance, nodeColor=<rgba 1 0.2603 0.2703 1>}
  NamedRerouteDeclaration(None=@n116.B) {name=VoronoiICenterDistance, nodeColor=<rgba 1 0.2603 0.2703 1>}
  n117 = NamedRerouteUsage()
  n118 = ScalarParameter() {defaultValue=3, desc=Metallic Band Amount, group=01_ComicShading, parameterName=Metallic_Bands}
  n119 = Multiply(A=Saturate(None=Power(Base=@n117) {constExponent=2.2}), B=@n118)
  n120 = Multiply(A=Saturate(None=ScalarParameter() {defaultValue=0.1, group=01_ComicShading, parameterName=Metallic_Roughness, sliderMax=1})) {constB=2}
  n121 = LinearInterpolate(
   B=Divide(
    A=MaterialFunctionCall!(In=@n119, Ramp Width=@n120) {materialFunction=/Engine/Functions/Engine_MaterialFunctions01/Math/SmoothCeil}.Result
    B=@n118
   )
   Alpha=ScalarParameter() {defaultValue=1, group=01_ComicShading, parameterName=Metallic_Intensity, sliderMax=1}
  ) {constA=1}
  n122 = ScalarParameter() {
   desc="The Bias will adjust the Metallic gradient.\r\n0.0 = the center matches the Color\r\n1.0 = the inner ring matches the Color\r\n2.0 = the outer ring match the Color"
   group=01_ComicShading
   parameterName=Metallic_Bias
   sliderMax=2
   sliderMin=-1
  }
  NamedRerouteDeclaration(None=Add(A=Saturate(None=@n121), B=Multiply(A=@n122) {constB=0.3333})) {name=Metallic_Gradient, nodeColor=<rgba 0.1107 1 0.7992 1>}
  n123 = LinearInterpolate(
   A=ScalarParameter() {defaultValue=5, group=01_ComicShading, parameterName=RimLight_FresnelExponent_Near, sliderMax=12}
   B=ScalarParameter() {defaultValue=4, group=01_ComicShading, parameterName=RimLight_FresnelExponent_Far, sliderMax=12}
   Alpha=NamedRerouteUsage()
  )
  n124 = NamedRerouteUsage()
  n125 = Saturate(None=InverseLinearInterpolate(
    A=ScalarParameter() {defaultValue=0.1, group=01_ComicShading, parameterName=RimLight_Min, sliderMax=1}
    B=ScalarParameter() {defaultValue=0.2, group=01_ComicShading, parameterName=RimLight_Max, sliderMax=1}
    Value=Fresnel(ExponentIn=@n123, Normal=@n124) {baseReflectFraction=0}
   ) {constA=0.1, constB=0.2})
  n126 = Saturate(None=NamedRerouteUsage())
  NamedRerouteDeclaration(None=StaticSwitchParameter(True=@n125, False=Multiply(A=@n126, B=@n125)) {group=01_ComicShading, parameterName=UseCameraBasedRimLight, sortPriority=0}) {name=RimLight, nodeColor=<rgba 1 0 0.1172 1>}
  n127 = NamedRerouteDeclaration(None=DotProduct(A=NamedRerouteUsage(), B=NamedRerouteUsage())) {name=NdotL_NoDN, nodeColor=<rgba 1 0.1814 0 1>}
  NamedRerouteDeclaration(None=Add(A=Multiply(A=@n127) {constB=0.5}) {constB=0.5}) {name=NdotL_Lambert_NoDN, nodeColor=<rgba 1 0.3522 0 1>}
  n128 = TextureCoordinate()
  n129 = MaterialFunctionCall(Color=NamedRerouteUsage()) {materialFunction=&SplitComponents}
  n130 = Saturate(None=Add(A=Add(A=@n129.R, B=@n129.R), B=@n129.B))
  n131 = ScalarParameter() {defaultValue=0.35, group=01_ComicShading, parameterName=Hatching_Line_Rot, sliderMax=1, sliderMin=-1}
  MaterialExpressionMultiply_11 = Multiply(A=@n131) {constB=2}
  n132 = Switch!(
   SwitchValue=StaticSwitchParameter(True=Round(None=@n130), False=Constant()) {defaultValue=true, group=01_ComicShading, parameterName=UseBaseColorForHatchingVariation, sortPriority=0}
   Default=@n131
   None=@MaterialExpressionMultiply_11
  ) {inputs [{
     inputName=None
     input {
      expression=&self:MaterialExpressionMultiply_11
      inputName=None
      mask=0
      maskA=0
      maskB=0
      maskG=0
      maskR=0
      outputIndex=0
     }
    }]}
  n133 = Multiply(
   A=Rotator(Coordinate=Add(A=@n128) {constB=2}, Time=@n132) {speed=3.1415}
   B=ScalarParameter() {defaultValue=64, group=01_ComicShading, parameterName=Hatching_Line_Amount, sliderMax=256}
  ) {constB=64}
  n134 = Abs(None=Add(A=Frac(None=@n133)) {constB=-0.5})
  n135 = OneMinus(None=Saturate(None=Multiply(A=@n134) {constB=2}))
  n136 = Floor(None=Add(A=@n133, B=Constant() {r=256}))
  n137 = Constant() {r=4}
  n138 = OneMinus(None=Saturate(None=Fmod(A=@n136, B=@n137))) {desc=1}
  n139 = MaterialFunctionCall(Float2=@n138) {materialFunction=&BreakOutFloat2Components}
  n140 = Constant() {r=2}
  n141 = OneMinus(None=Saturate(None=Fmod(A=@n136, B=@n140))) {desc=1}
  n142 = MaterialFunctionCall(Float2=Subtract(A=@n141, B=@n138)) {materialFunction=&BreakOutFloat2Components}
  n143 = MaterialFunctionCall(Float2=Subtract(A=Constant() {r=1}, B=@n141)) {materialFunction=&BreakOutFloat2Components}
  n144 = MaterialFunctionCall(Float3=Multiply(
    A=ComponentMask(None=@n135) {r=true}
    B=MaterialFunctionCall(X=@n139.R, Y=@n142.R, Z=@n143.R) {materialFunction=&MakeFloat3}.Result
   )) {materialFunction=&BreakOutFloat3Components}
  MaterialExpressionReroute_1 = Reroute(None=@n13)
  n145 = Frac(None=Add(A=@MaterialExpressionReroute_1.In) {constB=-0.5})
  MaterialExpressionReroute_3 = Reroute(None=@n14)
  n146 = OneMinus(None=@MaterialExpressionReroute_3.RampWidth)
  n147 = Max(A=Divide(A=Subtract(A=@n145, B=@n146), B=@MaterialExpressionReroute_3.RampWidth), B=Constant() {r=0.5}) {desc=It is good to make functions look like a lancer}
  n148 = Add(A=Round(None=@MaterialExpressionReroute_1.In), B=ConstantBiasScale(None=Cosine(None=@n147)))
  MaterialExpressionReroute_5 = Reroute(None=@n148)
  n149 = Add(A=@MaterialExpressionReroute_5.Output) {constB=-1}
  n150 = Max(A=Multiply(A=@n144.R, B=Saturate(None=@MaterialExpressionReroute_5.Output)), B=Multiply(A=@n144.G, B=Saturate(None=@n149)))
  n151 = Add(A=@n149) {constB=-1}
  n152 = Multiply(A=@n144.B, B=Saturate(None=@n151))
  n153 = MaterialFunctionCall(Float3=Multiply(
    A=ComponentMask(None=@n135) {g=true}
    B=MaterialFunctionCall(X=@n139.G, Y=@n142.G, Z=@n143.G) {materialFunction=&MakeFloat3}.Result
   )) {materialFunction=&BreakOutFloat3Components}
  n154 = Constant() {r=-1.1}
  n155 = Add(A=@n151, B=@n154) {constB=-1}
  n156 = Saturate(None=@n155)
  n157 = Add(A=@n155, B=@n154) {constB=-1}
  n158 = Max(A=Max(A=Max(A=@n150, B=@n152), B=Multiply(A=@n153.R, B=@n156)), B=Multiply(A=@n153.G, B=Saturate(None=@n157)))
  n159 = Multiply(A=@n153.B, B=Saturate(None=Add(A=@n157, B=@n154) {constB=-1}))
  NamedRerouteDeclaration(None=OneMinus(None=Max(A=@n158, B=@n159))) {name=LinearHatching, nodeColor=<rgba 1 0 0.0938 1>}
  n160 = TextureSampleParameter2D(UVs=@n9) {group=01_Anime Maps, parameterName=Color_Lit_Map, sortPriority=10, texture=&T_WhitesRGB}
  NamedRerouteDeclaration(None=@n160.RGB) {name=ColorLit, nodeColor=<rgba 0 0.3437 1 1>}
  n161 = TextureSampleParameter2D(UVs=@n9) {group=01_Anime Maps, parameterName=Color_Shaded_Map, sortPriority=11, texture=/Game/Global/Textures/Default/Blanks/T_MidgreysRGB}
  NamedRerouteDeclaration(None=Multiply(A=@n161.RGB, B=ScalarParameter() {defaultValue=1, group=01_ComicShading, parameterName=ColorShaded_Intensity})) {name=ColorShaded, nodeColor=<rgba 0.6641 1 0 1>}
  NamedRerouteUsage()
  PinBase() {reroutePins [
    In {expression=&self:MaterialExpressionReroute_1}
    RampWidth {expression=&self:MaterialExpressionReroute_3}
   ]}
  PinBase!(Output=@n148) {reroutePins [Output {expression=&self:MaterialExpressionReroute_5}]}
  TextureSampleParameter2D() {
   group=01_Anime Maps
   parameterName=Normal_Map
   samplerType=SAMPLERTYPE_Normal
   sortPriority=13
   texture=&FlatNormalBlue
  }
 ]}
