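# Parent material for the DriftFable cape cosmetic, giving it a comic-shaded look on skinned and cloth meshes.
# The DriftFable comic shading function, fed by shared eye UVs, produces material attributes that pass into the 2023 PostFX character parent function.
# A base-input composite supplies that function with a constant M, pre-skinned position (vertex-interpolated for pixel shaders), Fresnel, and distance blend from the shared character brightness/scale balance function.
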Material /BRCosmetics/Accessories/FORT_Capes/Cape_M_MED_DriftFable/Materials/M_ComicShader_DriftFable_Parent
graph {
 properties {
  bAutomaticallySetUsageInEditor=false
  bUseMaterialAttributes=true
  bUsedWithClothing=true
  bUsedWithMeshDeformer=true
  bUsedWithMorphTargets=true
  bUsedWithSkeletalMesh=true
  bUsedWithStaticLighting=true
  pixelDepthOffsetMode=PDOM_Legacy
  shadingModel=MSM_FromMaterialExpression
  usedShadingModels=Default Lit
 }
 nodes [
  Composite() {subgraphName=Construct Base Inputs}
  n1 = Constant()
  MaterialExpressionReroute_0 = Reroute(None=@n1)
  n2 = PreSkinnedPosition()
  n3 = ShaderStageSwitch(PixelShader=VertexInterpolator(VS=@n2).PS, VertexShader=@n2)
  MaterialExpressionReroute_1 = Reroute(None=@n3)
  n4 = MaterialFunctionCall!(Material) {materialFunction=/Game/Packages/Fortress_SharedMaterials/Base_Material_Functions/CharacterBrightnessandScale_Balance}
  MaterialExpressionReroute_2 = Reroute(None=@n4.Distance Blend)
  n5 = Fresnel() {baseReflectFraction=0, exponent=1}
  MaterialExpressionReroute_3 = Reroute(None=@n5)
  MaterialFunctionCall(
   Material=MaterialFunctionCall!(CustomUVs=MaterialFunctionCall() {materialFunction=/Game/Characters/Shared/FNEye/Materials/MF_EyeUVs}.EyeUVs) {materialFunction=&/MF_ComicShading_DriftFable}.Result
   M=@MaterialExpressionReroute_0.M
   Pre Skinned Local Position=@MaterialExpressionReroute_1.Pre Skinned Local Position
   Distance Blend=@MaterialExpressionReroute_2.Distance Blend
   Fresnel=@MaterialExpressionReroute_3.Fresnel
  ) {materialFunction=&MF_CharacterParent_2023_PostFX}
  PinBase()
  PinBase!(M=@n1, Pre Skinned Local Position=@n3, Distance Blend=@n4.Distance Blend, Fresnel=@n5) {reroutePins [
    M {expression=&self:MaterialExpressionReroute_0}
    Pre Skinned Local Position {expression=&self:MaterialExpressionReroute_1}
    Distance Blend {expression=&self:MaterialExpressionReroute_2}
    Fresnel {expression=&self:MaterialExpressionReroute_3}
   ]}
 ]
}
