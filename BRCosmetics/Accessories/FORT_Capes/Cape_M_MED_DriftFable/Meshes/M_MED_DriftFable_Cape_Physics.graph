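# Physics asset for the DriftFable medium male cape, defining bodies and joints for simulating the cape against the character.
# Four capsule chains, two outer from the clavicle_cape bones and two middle from MSTR bones, bend mostly on one swing axis with tight twist.
# Angularly free constraints cross-link the chains at mid-length, tying them into one sheet.
# Capsules on pelvis, spine, clavicles, upper arms and legs model the body, with notably oversized calf capsules.
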PhysicsAsset /BRCosmetics/Accessories/FORT_Capes/Cape_M_MED_DriftFable/Meshes/M_MED_DriftFable_Cape_Physics
bodies {
 backpack_root [{shapeType=Capsule, length=1, radius=1}]
 calf_l [{shapeType=Capsule, length=11.7963, radius=39.0787, center=<-26.1258 24.118 0>, rotation=<rot 90 179.8881 -178.6232>}]
 calf_r [{shapeType=Capsule, length=11.7963, radius=39.0787, center=<26.1258 -24.118 0>, rotation=<rot 90 180 1.4887>}]
 cape_01_l [{shapeType=Capsule, length=1, radius=1}]
 cape_01_r [{shapeType=Capsule, length=1, radius=1}]
 cape_02_l [{shapeType=Capsule, length=12, radius=9, center=<9 0 10>, rotation=<rot 90 0 0>}]
 cape_02_r [{shapeType=Capsule, length=12, radius=9, center=<9 0 10>, rotation=<rot 90 0 0>}]
 cape_03_l [{shapeType=Capsule, length=12, radius=12, center=<12 0 8.7996>, rotation=<rot 90 0 0>}]
 cape_03_r [{shapeType=Capsule, length=12, radius=12, center=<12 0 8.7996>, rotation=<rot 90 0 0>}]
 cape_04_l [{shapeType=Capsule, length=12, radius=12.723, center=<9 0 12.1972>, rotation=<rot 90 0 0>}]
 cape_04_r [{shapeType=Capsule, length=12, radius=12.723, center=<9 0 12.1276>, rotation=<rot 90 0 0>}]
 cape_05_l [{shapeType=Capsule, length=12, radius=12.723, center=<9 0 13.502>, rotation=<rot 90 0 0>}]
 cape_05_r [{shapeType=Capsule, length=12, radius=12.723, center=<9 0 13.502>, rotation=<rot 90 0 0>}]
 cape_06_l [{shapeType=Capsule, length=12, radius=12.723, center=<9 0 13.502>, rotation=<rot 90 0 0>}]
 cape_06_r [{shapeType=Capsule, length=12, radius=12.723, center=<9 0 13.502>, rotation=<rot 90 0 0>}]
 cape_07_l [{shapeType=Capsule, length=12, radius=12.723, center=<9 0 11.1651>, rotation=<rot 90 0 0>}]
 cape_07_r [{shapeType=Capsule, length=12, radius=12.723, center=<9 0 11.1651>, rotation=<rot 90 0 0>}]
 cape_mid_01_l [{shapeType=Capsule, length=12, radius=9, center=<9 0 10>, rotation=<rot 90 0 0>}]
 cape_mid_01_l_MSTR [{shapeType=Capsule, length=1, radius=1}]
 cape_mid_01_r [{shapeType=Capsule, length=12, radius=9, center=<9 0 10>, rotation=<rot 90 0 0>}]
 cape_mid_01_r_MSTR [{shapeType=Capsule, length=1, radius=1}]
 cape_mid_02_l [{shapeType=Capsule, length=12, radius=12, center=<12 0 10>, rotation=<rot 90 0 0>}]
 cape_mid_02_r [{shapeType=Capsule, length=12, radius=12, center=<12 0 10>, rotation=<rot 90 0 0>}]
 cape_mid_03_l [{shapeType=Capsule, length=12, radius=12.723, center=<9 0 13.502>, rotation=<rot 90 0 0>}]
 cape_mid_03_r [{shapeType=Capsule, length=12, radius=12.723, center=<10.4785 -0.2079 10.6126>, rotation=<rot 90 0 0>}]
 cape_mid_04_l [{shapeType=Capsule, length=12, radius=12.723, center=<9 0 13.502>, rotation=<rot 90 0 0>}]
 cape_mid_04_r [{shapeType=Capsule, length=12, radius=12.723, center=<9 0 13.502>, rotation=<rot 90 0 0>}]
 cape_mid_05_l [{shapeType=Capsule, length=12, radius=12.723, center=<9 0 13.502>, rotation=<rot 90 0 0>}]
 cape_mid_05_r [{shapeType=Capsule, length=12, radius=12.723, center=<9 0 11.6567>, rotation=<rot 90 0 0>}]
 cape_mid_06_l [{shapeType=Capsule, length=12, radius=12.723, center=<9 0 13.502>, rotation=<rot 90 0 0>}]
 cape_mid_06_r [{shapeType=Capsule, length=12, radius=12.723, center=<9 0 11.9495>, rotation=<rot 90 0 0>}]
 clavicle_cape_l [{shapeType=Capsule, length=1, radius=1}]
 clavicle_cape_r [{shapeType=Capsule, length=1, radius=1}]
 clavicle_l [{shapeType=Capsule, length=12.9688, radius=9.6562, center=<12.3638 -1.6165 0>, rotation=<rot 90 0 -6.5611>}]
 clavicle_r [{shapeType=Capsule, length=12.9688, radius=9.6562, center=<-12.3638 1.6165 0>, rotation=<rot 90 0 173.4389>}]
 pelvis [{shapeType=Capsule, length=3.8685, radius=14.2252, center=<-1.9314 -1.5382 0>}]
 spine_01 [{shapeType=Capsule, length=1.371, radius=14.6602, center=<0.5427 -0.4734 0>}]
 spine_02 [{shapeType=Capsule, length=2.6493, radius=13.1632, center=<3.5949 0.2148 0>}]
 spine_03 [{shapeType=Capsule, length=39.6562, radius=10, center=<2.5875 -1 0>}]
 spine_04 [{shapeType=Capsule, length=31.4062, radius=12, center=<2.8832 0.25 0>}]
 spine_05 [{shapeType=Capsule, length=32.875, radius=12, center=<-0.0029 0.1172 0>}]
 thigh_l [{shapeType=Capsule, length=23.2484, radius=14.0641, center=<-18.6538 -0.1273 0>, rotation=<rot 90 2.6271 0>}]
 thigh_r [{shapeType=Capsule, length=23.2484, radius=14.0641, center=<18.6538 0.1273 0>, rotation=<rot 90 -177.3729 0>}]
 upperarm_l [{shapeType=Capsule, length=20.2791, radius=7.2834, center=<11.9223 -1.2712 0>, rotation=<rot 90 6.5611 0>}]
 upperarm_r [{shapeType=Capsule, length=20.2791, radius=7.2834, center=<-11.9223 1.2712 0>, rotation=<rot 90 -173.4389 0>}]
}
constraints [
 {bone1Name=cape_mid_01_l, bone2Name=cape_mid_01_l_MSTR, swing1LimitDegrees=5, swing2LimitDegrees=45, twistLimitDegrees=5}
 {bone1Name=cape_mid_01_r, bone2Name=cape_mid_01_r_MSTR, swing1LimitDegrees=5, swing2LimitDegrees=45, twistLimitDegrees=5}
 {bone1Name=cape_01_l, bone2Name=clavicle_cape_l, swing1LimitDegrees=45, swing2LimitDegrees=45, twistLimitDegrees=5}
 {bone1Name=cape_02_l, bone2Name=cape_01_l, swing1LimitDegrees=5, swing2LimitDegrees=45, twistLimitDegrees=5}
 {bone1Name=cape_03_l, bone2Name=cape_02_l, swing1LimitDegrees=5, swing2LimitDegrees=45, twistLimitDegrees=5}
 {bone1Name=cape_04_l, bone2Name=cape_03_l, swing1LimitDegrees=10, swing2LimitDegrees=45, twistLimitDegrees=1}
 {bone1Name=cape_05_l, bone2Name=cape_04_l, swing1LimitDegrees=10, swing2LimitDegrees=45, twistLimitDegrees=1}
 {bone1Name=cape_06_l, bone2Name=cape_05_l, swing1LimitDegrees=10, swing2LimitDegrees=45, twistLimitDegrees=1}
 {bone1Name=cape_07_l, bone2Name=cape_06_l, swing1LimitDegrees=10, swing2LimitDegrees=45, twistLimitDegrees=1}
 {bone1Name=cape_01_r, bone2Name=clavicle_cape_r, swing1LimitDegrees=45, swing2LimitDegrees=45, twistLimitDegrees=5}
 {bone1Name=cape_02_r, bone2Name=cape_01_r, swing1LimitDegrees=5, swing2LimitDegrees=45, twistLimitDegrees=5}
 {bone1Name=cape_03_r, bone2Name=cape_02_r, swing1LimitDegrees=5, swing2LimitDegrees=45, twistLimitDegrees=5}
 {bone1Name=cape_04_r, bone2Name=cape_03_r, swing1LimitDegrees=10, swing2LimitDegrees=45, twistLimitDegrees=1}
 {bone1Name=cape_05_r, bone2Name=cape_04_r, swing1LimitDegrees=10, swing2LimitDegrees=45, twistLimitDegrees=1}
 {bone1Name=cape_06_r, bone2Name=cape_05_r, swing1LimitDegrees=10, swing2LimitDegrees=45, twistLimitDegrees=1}
 {bone1Name=cape_07_r, bone2Name=cape_06_r, swing1LimitDegrees=10, swing2LimitDegrees=45, twistLimitDegrees=1}
 {bone1Name=cape_mid_02_l, bone2Name=cape_mid_01_l, swing1LimitDegrees=5, swing2LimitDegrees=45, twistLimitDegrees=5}
 {bone1Name=cape_mid_03_l, bone2Name=cape_mid_02_l, swing1LimitDegrees=10, swing2LimitDegrees=45, twistLimitDegrees=1}
 {bone1Name=cape_mid_04_l, bone2Name=cape_mid_03_l, swing1LimitDegrees=10, swing2LimitDegrees=45, twistLimitDegrees=1}
 {bone1Name=cape_mid_05_l, bone2Name=cape_mid_04_l, swing1LimitDegrees=10, swing2LimitDegrees=45, twistLimitDegrees=1}
 {bone1Name=cape_mid_06_l, bone2Name=cape_mid_05_l, swing1LimitDegrees=10, swing2LimitDegrees=45, twistLimitDegrees=1}
 {bone1Name=cape_mid_02_r, bone2Name=cape_mid_01_r, swing1LimitDegrees=5, swing2LimitDegrees=45, twistLimitDegrees=5}
 {bone1Name=cape_mid_03_r, bone2Name=cape_mid_02_r, swing1LimitDegrees=10, swing2LimitDegrees=45, twistLimitDegrees=1}
 {bone1Name=cape_mid_04_r, bone2Name=cape_mid_03_r, swing1LimitDegrees=10, swing2LimitDegrees=45, twistLimitDegrees=1}
 {bone1Name=cape_mid_05_r, bone2Name=cape_mid_04_r, swing1LimitDegrees=10, swing2LimitDegrees=45, twistLimitDegrees=1}
 {bone1Name=cape_mid_06_r, bone2Name=cape_mid_05_r, swing1LimitDegrees=10, swing2LimitDegrees=45, twistLimitDegrees=1}
 {
  bone1Name=cape_mid_04_l
  bone2Name=cape_mid_04_r
  swing1LimitDegrees=45
  swing1Motion=Free
  swing2LimitDegrees=45
  swing2Motion=Free
  twistLimitDegrees=45
  twistMotion=Free
 }
 {
  bone1Name=cape_05_l
  bone2Name=cape_mid_04_l
  swing1LimitDegrees=45
  swing1Motion=Free
  swing2LimitDegrees=45
  swing2Motion=Free
  twistLimitDegrees=45
  twistMotion=Free
 }
 {
  bone1Name=cape_mid_04_r
  bone2Name=cape_05_r
  swing1LimitDegrees=45
  swing1Motion=Free
  swing2LimitDegrees=45
  swing2Motion=Free
  twistLimitDegrees=45
  twistMotion=Free
 }
]
